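# Animation Blueprint for the Genius StandAlone back bling, a pack whose face reacts to the wearer's kills and health loss.
# A kill-stat watcher and a per-update health comparison trigger additive Smile or Angry animations and swap "Hide Element" material parameters, restoring the default after a retriggerable delay.
# The additive state machine layers over AnimDynamics on two bones, then a cached pose blends with an offset pack pose selected by the targeting flag.
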AnimBlueprint /BRCosmetics/Accessories/FORT_Backpacks/Backpack_M_MED_Genius_StandAlone/Meshes/M_MED_Genius_StandAlone_Pack_AnimBP
target_skeleton=&/M_MED_Genius_StandAlone_Pack_Skeleton
graphs [
 AnimGraph {nodes [
   n1 = "Transform (Modify) Bone - Bone: pack"(ComponentPose=Local To Component(LocalPose=Use cached pose 'normal'()), Translation="0,0,3", Rotation="-0.000,-35,-10") {
    rotationMode=BMM_Additive
    translationMode=BMM_Additive
    alphaBoolBlend {blendOption=Linear}
    boneToModify {boneName=pack}
    rotation {roll=-10, yaw=-35}
    translation {z=3}
   }
   n2 = Component To Local(ComponentPose=@n1)
   n3 = Use cached pose 'normal'()
   n4 = Get bIsTargeting()
   n5 = "Transform (Modify) Bone - Bone: root"(
    ComponentPose=Local To Component(LocalPose=Blend Poses by bool(BlendPose_0=@n2, BlendPose_1=@n3, bActiveValue=@n4) {blendType=Linear})
    Translation="0,0.5,0"
    Rotation="0,0,0"
   ) {
    translationMode=BMM_Additive
    translationSpace=BCS_BoneSpace
    alphaBoolBlend {blendOption=Linear}
    boneToModify {boneName=root}
    translation {y=0.5}
   }
   Output Pose(Result=Component To Local(ComponentPose=@n5)) {layerGroup=None}
   n6 = Anim Dynamics - Bone: dyn_lower(ComponentPose=Local To Component(LocalPose=Local Space Ref Pose()), Alpha=0.4, ExternalForce="0,0,0") {
    angularDampingOverride=0.9
    angularSpringConstant=20000
    bAngularSpring=true
    bOverrideAngularDamping=true
    bUseGravityOverride=true
    simulationSpace=RootRelative
    alphaBoolBlend {blendOption=Linear}
    boundBone {boneName=dyn_lower}
    componentAppliedLinearAccClamp=<3000 3000 3000>
    componentLinearAccScale=<0.3 0.3 0.3>
    componentLinearVelScale=<0.3 0.3 0.3>
    gravityOverride {z=-980}
    physicsBodyDefinitions [{
      boundBone {boneName=dyn_lower}
      boxExtents=<40 40 40>
      constraintSetup {angularLimitsMax=<15 4 4>, angularLimitsMin=<-15 -4 -4>, angularTarget=<0 0 0>}
      localJointOffset=<13 0 0>
     }]
   }
   n7 = Anim Dynamics - Bone: dyn_upper(ComponentPose=@n6, Alpha=0.4, ExternalForce="0,0,0") {
    angularDampingOverride=0.99
    angularSpringConstant=100
    bAngularSpring=true
    bOverrideAngularDamping=true
    bUseGravityOverride=true
    simulationSpace=RootRelative
    alphaBoolBlend {blendOption=Linear}
    boundBone {boneName=dyn_upper}
    componentAppliedLinearAccClamp=<3000 3000 3000>
    componentLinearAccScale=<0.3 0.3 0.3>
    componentLinearVelScale=<0.3 0.3 0.3>
    gravityOverride {z=-980}
    physicsBodyDefinitions [{
      boundBone {boneName=dyn_upper}
      boxExtents=<40 40 40>
      constraintSetup {angularLimitsMax=<15 5 15>, angularLimitsMin=<-15 -5 -15>, angularTarget=<0 0 0>}
      localJointOffset=<6 0 0>
     }]
   }
   n8 = "Transform (Modify) Bone - Bone: pack"(ComponentPose=@n7, Translation="0,-3,0", Rotation="-0.000,-30,0.000") {
    rotationMode=BMM_Replace
    translationMode=BMM_Replace
    alphaBoolBlend {blendOption=Linear}
    boneToModify {boneName=pack}
    rotation {yaw=-30}
    translation {y=-3}
   }
   AnimGraphNode_StateMachine_0 = Reactivity() {is=title AnimGraphNode_StateMachine}
   Save cached pose 'normal'(Pose=Apply Additive(Base=Component To Local(ComponentPose=@n8), Additive=@AnimGraphNode_StateMachine_0, Alpha=0.95) {alphaBoolBlend {blendOption=Linear}}) {cacheName=normal}
  ]}
 Reactivity {
  entry=Idle
  path=AnimGraph.AnimGraphNode_StateMachine_0.Reactivity
  ids=id
  nodes [
   AnimStateNode_0 = Idle()
   AnimStateNode_1 = Smile() {is=title AnimStateNode, bAlwaysResetOnEntry=true}
   AnimStateNode_2 = Angry() {is=title AnimStateNode}
   AnimStateTransitionNode_0 = AnimStateTransitionNode()
   AnimStateTransitionNode_1 = AnimStateTransitionNode()
   AnimStateTransitionNode_2 = AnimStateTransitionNode()
   AnimStateTransitionNode_3 = AnimStateTransitionNode()
  ]
  links [
   {
    dst=Idle
    rule=AnimGraph.AnimGraphNode_StateMachine_0.Reactivity.AnimStateTransitionNode_3.Transition
    src=Angry
    via=AnimStateTransitionNode_3
   }
   {
    dst=Angry
    rule=AnimGraph.AnimGraphNode_StateMachine_0.Reactivity.AnimStateTransitionNode_2.Transition
    src=Idle
    via=AnimStateTransitionNode_2
   }
   {
    dst=Smile
    rule=AnimGraph.AnimGraphNode_StateMachine_0.Reactivity.AnimStateTransitionNode_0.Transition
    src=Idle
    via=AnimStateTransitionNode_0
   }
   {
    dst=Idle
    rule=AnimGraph.AnimGraphNode_StateMachine_0.Reactivity.AnimStateTransitionNode_1.Transition
    src=Smile
    via=AnimStateTransitionNode_1
   }
  ]
 }
 Idle {path=AnimGraph.AnimGraphNode_StateMachine_0.Reactivity.AnimStateNode_0.Idle, nodes [Output Animation Pose()]}
 Smile {
  path=AnimGraph.AnimGraphNode_StateMachine_0.Reactivity.AnimStateNode_1.Smile
  nodes [
   n1 = "Sequence Player (additive) 'Genius_StandAlone_Smile_Pickle'"() {
    bLoopAnimation=false
    sequence=/BRCosmetics/Accessories/FORT_Backpacks/Backpack_M_MED_Genius_StandAlone/Animation/Genius_StandAlone_Smile_Pickle
   }
   Output Animation Pose(Result=@n1)
  ]
 }
 Angry {
  path=AnimGraph.AnimGraphNode_StateMachine_0.Reactivity.AnimStateNode_2.Angry
  nodes [
   n1 = "Sequence Player (additive) 'Genius_StandAlone_Angry_Pickle'"() {
    bLoopAnimation=false
    sequence=/BRCosmetics/Accessories/FORT_Backpacks/Backpack_M_MED_Genius_StandAlone/Animation/Genius_StandAlone_Angry_Pickle
   }
   Output Animation Pose(Result=@n1)
  ]
 }
 Transition {
  path=AnimGraph.AnimGraphNode_StateMachine_0.Reactivity.AnimStateTransitionNode_0.Transition
  nodes [Result(bCanEnterTransition=Get SmileAnimation() -> (SmileAnimation:bool))]
 }
 Transition {
  path=AnimGraph.AnimGraphNode_StateMachine_0.Reactivity.AnimStateTransitionNode_1.Transition
  nodes [Result(bCanEnterTransition="float <= float"(A="Time Remaining (Genius_StandAlone_Smile_Pickle)"!() -> (ReturnValue:float) {is=title K2Node_AnimGetter}, B=0.1) -> (ReturnValue=~))]
 }
 Transition {
  path=AnimGraph.AnimGraphNode_StateMachine_0.Reactivity.AnimStateTransitionNode_2.Transition
  nodes [Result(bCanEnterTransition=Get AngryAnimation() -> (AngryAnimation:bool))]
 }
 Transition {
  path=AnimGraph.AnimGraphNode_StateMachine_0.Reactivity.AnimStateTransitionNode_3.Transition
  nodes [Result(bCanEnterTransition="float <= float"(A="Time Remaining (Genius_StandAlone_Angry_Pickle)"!() -> (ReturnValue:float) {is=title K2Node_AnimGetter}, B=0.1) -> (ReturnValue=~))]
 }
 EventGraph {nodes [
   Event BlueprintInitializeAnimation()
   > Set Skel_Mesh(Skel_Mesh=GetOwnerSkeletalMeshComponent())
   Event BlueprintPostEvaluateAnimation() {is=title K2Node_Event}
   > Set SmileAnimation(SmileAnimation:bool=false) -> (Output_Get:bool)
   > Set AngryAnimation(AngryAnimation:bool=false) -> (Output_Get:bool)
   Event BlueprintUpdateAnimation()
   > n1 = Sequence!(execute) -> (then_0, then_1, then_2:Exec, then_3:Exec)
   > n2 = Get!(execute:Exec) -> (then:Exec, else:Exec, pawn:Fort Player Pawn Object Reference)
   n3 = Cast To FortPlayerPawn(execute=[@n11.then, @n2.else], Object=TryGetFortPawn())
   > Set pawn(pawn:Fort Player Pawn Object Reference=@n3) -> (Output_Get:Fort Player Pawn Object Reference)
   > n4 = WatchCosmeticStat!(
    execute:Exec
    Pawn:Fort Player Pawn Object Reference=Get pawn() -> (pawn:Fort Player Pawn Object Reference)
    StatTag:Fort Stat Manager Tag Structure=(TagName="StatManager.AthenaKills")
    DefaultValue:int=0
    bBroadcastInitialValue:bool=false
    InitialValueBroadcastDelay:float=-1
    ChangeValueBroadcastDelay:float=-1
    bConsumeWaitingNotifications:bool=true
   ) -> (then:Exec, OnStatChanged:Exec, StatName:Name, StatValue:int) {is=title K2Node_AsyncAction}
   n5 = "To Float (Integer)"(InInt=@n4.StatValue)
   n6 = "float > float"(A=@n5, B=Get KillCount() -> (KillCount:float)) {is=title K2Node_CallFunction}
   Branch(execute=@n4.OnStatChanged, Condition=@n6)
   > Set Active_Kill(Active_Kill:bool=@n6) -> (Output_Get:bool)
   > Set KillCount(KillCount:float=@n5) -> (Output_Get:float)
   > Branch(Condition=Get Active_Kill() -> (Active_Kill:bool))
   > Set SmileAnimation(SmileAnimation=true)
   > n7 = Get Dynamic Material Instance(SkeletalMeshComponent=GetPartSkeletalMeshComponent())
   > SetScalarParameterValue(self=@n7, ParameterName=Hide Element 03, Value=1)
   > SetScalarParameterValue(self=@n7, ParameterName=Hide Element 01, Value=1)
   > SetScalarParameterValue(self=@n7, ParameterName=Hide Element 02, Value=0)
   > RetriggerableDelay(Duration=1.2)
   > n8 = Set Active_Kill(Active_Kill=false)
   n9 = Get Dynamic Material Instance(execute=@n1.then_1, SkeletalMeshComponent=GetPartSkeletalMeshComponent())
   SetScalarParameterValue(execute=[@n14.then, @n8.then], self=[@n7, @n9], ParameterName=Hide Element 03, Value=1)
   > SetScalarParameterValue(self=[@n7, @n9], ParameterName=Hide Element 02, Value=1)
   > SetScalarParameterValue(self=[@n7, @n9], ParameterName=Hide Element 01, Value=0)
   Event OnCharacterCustomizationLoaded()
   > n10 = Get Dynamic Material Instance(SkeletalMeshComponent=GetPartSkeletalMeshComponent())
   > SetScalarParameterValue(self=@n10, ParameterName=Hide Element 03, Value=1)
   > SetScalarParameterValue(self=@n10, ParameterName=Hide Element 02, Value=1)
   > n11 = SetScalarParameterValue(self=@n10, ParameterName=Hide Element 01, Value=0)
   n12 = Cast To FortPlayerPawnAthena(execute=@n9.then, Object=TryGetPawnOwner())
   n13 = GetHealth!(self:Fort Pawn Object Reference=@n12) -> (ReturnValue:float)
   Set healthVar(execute=@n15.else, healthVar:double=@n13) -> (Output_Get:double)
   SetScalarParameterValue(execute=@n16.then, self=@n9, ParameterName=Hide Element 02, Value=1)
   > SetScalarParameterValue(self=@n9, ParameterName=Hide Element 01, Value=1)
   > SetScalarParameterValue(self=@n9, ParameterName=Hide Element 03, Value=0)
   > n14 = RetriggerableDelay(Duration=1.2)
   n15 = Branch(execute=@n12.then, Condition="float < float"(A=@n13, B=Get healthVar() -> (healthVar:double)))
   > Set healthVar(healthVar=@n13)
   > n16 = Set AngryAnimation(AngryAnimation=true)
  ]}
]
variables [
 Active_Kill {type=boolean}
 AngryAnimation {type=boolean}
 DefaultFace {type=boolean}
 Did_Kill {type=boolean}
 head_skelmesh {type=SkeletalMeshComponent}
 healthVar {type=number}
 Is Targeting {type=boolean}
 killCount {type=number}
 Last_KillCount {type=number}
 Pawn {type=FortPlayerPawn}
 Skel_Mesh {type=SkeletalMeshComponent}
 SmileAnimation {type=boolean}
]
